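# UI material that draws an animated, glowing circular outline, named as the outline state of a Creator tile in the Discovery Browser.
# An SDF circle feeds two outline passes, a crisp line and a soft, faint glow, with its size inset to leave room for them.
# A time-rotated UV gradient, mirrored into a triangle wave, modulates brightness so the highlight flows.
# The result is tinted by a Color parameter and scaled by vertex alpha.
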Material /DiscoveryBrowser/Tiles/Creator/Materials/M_UI_Creator_Tile_OutlineState
graph {
 properties {
  bAutomaticallySetUsageInEditor=false
  bUsedWithStaticMesh=false
  blendMode=BLEND_AlphaComposite
  materialDomain=MD_UI
  usedShadingModels=Default Lit
 }
 nodes [
  n1 = Multiply(
   A=Time()
   B=NamedRerouteDeclaration(None=ScalarParameter() {defaultValue=0.15, parameterName=FlowSpeed}) {name=Speed, nodeColor=<rgba 0 0 0 1>}
  ) {constB=0.1}
  n2 = NamedRerouteDeclaration(None=MaterialFunctionCall() {materialFunction=&GetUserInterfaceUV}.Normalized UV) {name=UV, nodeColor=<rgba 0 0 0 1>}
  n3 = Saturate(None=ComponentMask(None=Rotator(Coordinate=@n2, Time=@n1) {speed=6.2832}) {r=true})
  n4 = Multiply(A=Min(A=@n3, B=OneMinus(None=@n3))) {constB=2}
  n5 = LinearInterpolate(Alpha=SmoothStep(Value=@n4) {constMax=0.75}) {constA=0.125}
  n6 = Constant()
  n7 = ScalarParameter() {defaultValue=0.1, parameterName=Glow Range, sliderMax=1}
  n8 = NamedRerouteUsage()
  n9 = Add(A=@n8, B=Multiply(A=NamedRerouteUsage()) {constB=0.5})
  n10 = MaterialFunctionCall(UVs=@n2, Size=OneMinus(None=Max(A=@n8, B=@n9))) {materialFunction=&MF_UI_Circle}
  n11 = MaterialFunctionCall(
   Input=MaterialFunctionCall!(
    DistanceField=@n10.SDF
    Threshold
    Thickness=NamedRerouteUsage()
    Alignment=@n6
    Smoothing=NamedRerouteDeclaration(None=@n7) {name=Glow Range, nodeColor=<rgba 0 0 0 1>}
   ) {materialFunction=/FortUILibrary/MaterialFunctions/Shapes/Modifiers/MF_UI_ShapeOutline}.Outline
   Input Low=Constant()
   Input High=Constant() {r=0.5}
  ) {materialFunction=&RemapValue_0-1}
  n12 = ScalarParameter() {defaultValue=0.0175, parameterName=Outline Thickness, sliderMax=0.5}
  n13 = LinearInterpolate(
   A=Multiply(A=LinearInterpolate(Alpha=@n4) {constA=0.125, constB=0.25}, B=SmoothStep(Value=@n11.Result)) {constB=0.125, desc=Glow Opacity}
   Alpha=MaterialFunctionCall(DistanceField=@n10.SDF, Thickness=NamedRerouteDeclaration(None=@n12) {name=Outline Thickness, nodeColor=<rgba 0 0 0 1>}, Alignment=@n6) {materialFunction=&MF_UI_ShapeOutline}.Outline
  )
  Multiply(
   A=Multiply(A=Multiply(A=@n5, B=@n13), B=VectorParameter() {parameterName=Color, defaultValue=<rgba 0.6724 0 0.0027 1>}.RGB)
   B=VertexColor().A
  )
 ]
}
