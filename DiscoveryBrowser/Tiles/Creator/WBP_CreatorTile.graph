WidgetBlueprint /DiscoveryBrowser/Tiles/Creator/WBP_CreatorTile
animations [Anim_OnLoadedImage {binding_count=0, length_seconds=5, object=&self:Anim_OnLoadedImage, sequence {bindings []}}]
graphs [
 EventGraph {nodes [
   OnTileSizeSetOverrideContainerSize() {is=title K2Node_CustomEvent}
   n1 = Get WBP_Discover_Tile_Base() -> (WBP_Discover_Tile_Base:WBP Discover Tile Base Object Reference)
   n2 = BreakVector2D(InVec=Get Dynamic Tile Size!(self:WBP Discover Tile Base Object Reference=@n1) -> (DynamicTileSize:Vector 2D Structure) {is=title K2Node_CallFunction})
   > Set Tile Container Size Override!(
    execute
    self:WBP Discover Tile Base Object Reference=@n1
    TileSizeOverrideWithOutsideContent:Vector 2D Structure=MakeVector2D(X=@n2, Y=float + float(A=@n2.Y, B=30))
   ) -> (then) {is=title K2Node_CallFunction}
   n3 = OnModalDismissed!() -> (OutputDelegate, then, ButtonIndex:int) {is=title K2Node_CustomEvent}
   n4 = Construct Object from Class(execute=@n7.then, Class=/DiscoveryBrowser/CreatorProfile/Dialogs/DialogVM_FollowLimitReached.DialogVM_FollowLimitReached_C) -> (ReturnValue:Dialog VM Follow Limit Reached Object Reference)
   > n5 = Create Widget!(
    execute:Exec
    Class:User Widget Class Reference=/FortUILibrary/UIKit/Dialogs/WBP_UIKit_Dialog.WBP_UIKit_Dialog_C
    OwningPlayer:Player Controller Object Reference
    UIKitDialogViewModel:UI Kit Dialog Viewmodel Object Reference=@n4
   ) -> (then:Exec, ReturnValue:WBP UIKit Dialog Object Reference)
   n6 = Get Dialog_Internal!(self:WBP UIKit Dialog Object Reference=@n5) -> (Dialog_Internal:WBP UIKit Dialog Internal Object Reference)
   Unbind all Events from Dialog Button Clicked Dispatcher!(execute:Exec=@n3.then, self:WBP UIKit Dialog Internal Object Reference=@n6) -> (then:Exec) {is=title K2Node_ClearDelegate}
   > DeactivateWidget(self=@n5)
   n7 = OnReachedMaxNumberOfFollowedCreators() {is=title K2Node_CustomEvent}
   n8 = ToggleFollowing!() -> (OutputDelegate, then, InteractionNameID:Name) {is=title K2Node_CustomEvent}
   n9 = Get FortPlayerPageProfileVM() -> (FortPlayerPageProfileVM:Creator View Model Object Reference)
   > ToggleAmIFollowing(self:Creator View Model Object Reference=@n9)
   > n10 = Branch(Condition=Get bAmIFollowing!(self:Creator View Model Object Reference=@n9) -> (bAmIFollowing:bool))
   > n11 = PlaySound2D(Sound=Get Favorited_Sound() -> (Favorited_Sound:Sound Base Object Reference))
   CloseContextMenu(execute=[@n11.then, @n12.then], self:WBP Discover Tile Base Object Reference=Get WBP_Discover_Tile_Base())
   n12 = PlaySound2D(execute=@n10.else, Sound=Get Unfavorited_Sound() -> (Unfavorited_Sound:Sound Base Object Reference))
   n13 = AddFollowUnfollowToContextMenu!() -> (OutputDelegate, then, IsFollowing:bool=~) {is=title K2Node_CustomEvent}
   > RemoveInteractionByName!(
    execute
    self:WBP Discover Tile Base Object Reference=Get WBP_Discover_Tile_Base()
    ActionToRemove:Name=Get ContextMenuFollowName() -> (ContextMenuFollowName:Name=None)
   ) -> (then)
   n14 = GetRuntimeOptions()
   > AddInteraction!(
    execute
    self:UIKit Context Menu Action Context Object Reference=Get ContextMenuActionContext!(self:WBP Discover Tile Base Object Reference=Get WBP_Discover_Tile_Base()) -> (ContextMenuActionContext:UIKit Context Menu Action Context Object Reference)
    InActionInfo_DisplayName:Text=SelectText(
     A=SelectText(A="LOCTEXT(\"Unfollow\")", B="LOCTEXT(\"Follow\")", bPickA=@n13.IsFollowing)
     B=SelectText(A="LOCTEXT(\"Unfavorite\")", B="LOCTEXT(\"Favorite\")", bPickA=@n13.IsFollowing)
     bPickA=GetIsCreatorFollowRebrandingEnabled!(self:Fort Runtime Options Object Reference=@n14) -> (ReturnValue:bool)
    )
    InActionInfo_IconTexture:Texture 2D Soft Object Reference
    InActionInfo_IconMaterial:Material Instance Soft Object Reference
    InActionInfo_SortPriority:int=1000
    InActionInfo_bIsAvailable:bool=True
    InActionInfo_bIsEnabled:bool=True
    InActionInfo_bIsSelected:bool=False
    InActionInfo_bAddDivider:bool=False
    InActionInfo_bPositionDividerAfter:bool=False
    InActionInfo_bCloseMenuOnSelect:bool=False
    InActionInfo_InteractionNameID:Name=Get ContextMenuFollowName() -> (ContextMenuFollowName=None)
    InActionInfo_Interaction:Data Table Row Handle Structure=(DataTable=None,RowName="")
    InActionInfo_PressedDelegate:Delegate="(null).None"
    InActionInfo_ReleasedDelegate:Delegate="(null).None"
    InInteractionDelegate:"Delegate (by ref)"=@n8
   ) -> (then)
   n15 = Event PreConstruct()
   > Parent: PreConstruct(IsDesignTime=@n15.IsDesignTime)
   n16 = Get WBP_Discover_Tile_Base()
   > UpdatePrimaryTextBlock!(execute, self:WBP Discover Tile Base Object Reference=@n16, UsePrimaryText:bool=true) -> (then)
   > UpdateSecondaryTextBlock!(execute, self:WBP Discover Tile Base Object Reference=@n16, UseSecondaryText:bool=true) -> (then)
   > Set Tile Size!(execute, self:WBP Discover Tile Base Object Reference=@n16, InSize:Vector 2D Structure=(X=192,Y=192)) -> (then) {is=title K2Node_CallFunction}
   > SetSecondaryText!(execute, self:WBP Discover Tile Base Object Reference=@n16, SecondaryText:Text="LOCTEXT(\"Playing Now\")") -> (then)
   > UpdateSecondaryTextColorAndOpacity!(
    execute
    self:WBP Discover Tile Base Object Reference=@n16
    UseDefault:bool=false
    OverrideColor:Slate Color Structure=Make Slate Color(SpecifiedColor=(R=0.1128,G=0.6562,B=0.3203,A=1))
   ) -> (then)
   > SetOutlineSquareOrCircular!(execute, self:WBP Discover Tile Base Object Reference=@n16, IsOutlineCircular:bool=true) -> (then)
   > UpdateSecondaryTextJustification!(execute, self:WBP Discover Tile Base Object Reference=@n16, Justification:ETextJustify Enum=Center) -> (then)
   > UpdatePrimaryTextJustification!(execute, self:WBP Discover Tile Base Object Reference=@n16, Justification:ETextJustify Enum=Center) -> (then)
   > SetCompactUIStyle() {is=title K2Node_CallFunction}
   > SetVisibility(self=Get WBP_PlusFollowButton() -> (WBP_PlusFollowButton:WBP Creator Following Indicator Object Reference), InVisibility=Collapsed)
   Event Construct()
   > Parent: Construct()
   n17 = Get WBP_Discover_Tile_Base()
   > Bind Event to On Set Tile Container Size Override(self:WBP Discover Tile Base Object Reference=@n17, Delegate=Create Event()) {is=title K2Node_AddDelegate}
   Event On Clicked()
   > Parent: On Clicked() {is=title K2Node_CallParentFunction}
   n18 = GetIdString!(self:Creator View Model Object Reference=Get FortPlayerPageProfileVM()) -> (ReturnValue:String)
   > SendFlowRiderClickTileEvent!(
    execute
    self:WBP Discover Tile Base Object Reference=Get WBP_Discover_Tile_Base()
    TileWidget:User Widget Object Reference=Self-Reference()
    TileType:EActivityItemType Enum=Creator
    TileUniqueIdentifier:String=@n18
   ) -> (then)
   n19 = Get FortActivitySelector() -> (ReturnValue:Fort Activity Selector Object Reference) {is=title K2Node_GetSubsystem}
   > HandleCreatorTileClicked!(
    execute
    self:Fort Activity Selector Object Reference=@n19
    ViewModel:Discover Tile Item View Model Object Reference=Get FortPlayerPageProfileVM()
   ) -> (then)
   > OpenCreatorPageFromCreatorId!(execute, self:Fort Activity Selector Object Reference=@n19, InNewCreatorId:String=@n18) -> (then)
   Event Destruct()
   > Parent: Destruct()
   > Unbind all Events from On Set Tile Container Size Override!(execute:Exec, self:WBP Discover Tile Base Object Reference=@n17) -> (then:Exec) {is=title K2Node_ClearDelegate}
   > Unbind all Events from On Reached Max Number Of Followed Creators!(execute:Exec, self:Creator View Model Object Reference=Get FortPlayerPageProfileVM()) -> (then:Exec) {is=title K2Node_ClearDelegate}
   n20 = Event SetNestedButtonGroupSelectionVM!() -> (OutputDelegate, then, InVM:Fort Nested Button Group Selection VM Object Reference) {is=title K2Node_Event}
   > Set FortNestedButtonGroupSelectionVM(FortNestedButtonGroupSelectionVM:Fort Nested Button Group Selection VM Object Reference=@n20.InVM) -> (Output_Get:Fort Nested Button Group Selection VM Object Reference)
   n21 = Event OnListItemObjectSet()
   > n22 = Cast To FortPlayerPageProfileVM!(execute, Object=@n21.ListItemObject) -> (then, CastFailed, AsCreator View Model:Creator View Model Object Reference) {is=title K2Node_DynamicCast}
   > n23 = Set FortPlayerPageProfileVM(FortPlayerPageProfileVM:Creator View Model Object Reference=@n22) -> (Output_Get:Creator View Model Object Reference)
   n24 = Get FortPlayerPageProfileVM()
   > ConfigureDebugUI!(
    execute
    self:WBP Discover Tile Base Object Reference=Get WBP_Discover_Tile_Base()
    Debug VM:Discover Tile Item Debug View Model Object Reference=Get DebugVM!(self:Discover Tile Item View Model Object Reference=@n24) -> (DebugVM:Discover Tile Item Debug View Model Object Reference)
    OwningButton:Common Button Base Object Reference=Self-Reference()
   ) -> (then)
   > SetParentIDForContextMenuTelemetry!(
    execute
    self:WBP UIKit Context Menu Anchor Button Base Object Reference=Get WBP_UIKit_ContextMenuAnchorButton!(self:WBP Discover Tile Base Object Reference=Get WBP_Discover_Tile_Base()) -> (WBP_UIKit_ContextMenuAnchorButton:WBP UIKit Context Menu Anchor Button Object Reference)
    ParentID:String=Append(A=Creator ID:, B=GetIdString(self=@n24))
    IsParentIdEGPI:bool=true
   ) -> (then)
   > SetTileItemViewModel!(
    execute
    self:WBP Discover Tile Base Object Reference=Get WBP_Discover_Tile_Base()
    TileItemViewModel:Discover Tile Item View Model Object Reference=@n23
   ) -> (then)
   n25 = Get FortPlayerPageProfileVM()
   > StartVisibilityImpression(self:Discover Tile Item View Model Object Reference=@n25)
   > Bind Event to On Reached Max Number Of Followed Creators(self:Creator View Model Object Reference=@n25, Delegate=@n7) {is=title K2Node_AddDelegate}
   TryPushWidgetAsModalPanel!(execute=@n26.then, PushWidget:User Widget Object Reference=@n5) -> (then)
   n26 = Bind Event to Dialog Button Clicked Dispatcher(execute=@n5.then, self:WBP UIKit Dialog Internal Object Reference=@n6, Delegate=@n3) {is=title K2Node_AddDelegate}
  ]}
 GetNestedButtonGroupSelectionVM {nodes [
   GetNestedButtonGroupSelectionVM() {is=title K2Node_FunctionEntry}
   > Return Node!(
    execute
    ReturnValue:Fort Nested Button Group Selection VM Object Reference=Get FortNestedButtonGroupSelectionVM() -> (FortNestedButtonGroupSelectionVM:Fort Nested Button Group Selection VM Object Reference)
   )
  ]}
 LoadAvatarImage {nodes [
   LoadAvatarImage!() -> (then, Url:String) {is=title K2Node_FunctionEntry}
   > LoadAvatarImage(self:Creator View Model Object Reference=Get FortPlayerPageProfileVM()) {is=title K2Node_CallFunction}
  ]}
 SetColumnSize {nodes [
   n1 = SetColumnSize!() -> (then, ColSize:int) {is=title K2Node_FunctionEntry}
   n2 = Get WBP_Discover_Tile_Base()
   > Return Node!(
    execute
    TileSize:Vector 2D Structure=Select(
     Option 0:Vector 2D Structure=Get TileSize!(self:WBP Discover Tile Base Object Reference=@n2) -> (TileSize:Vector 2D Structure)
     Option 1:Vector 2D Structure=ColumnSizeToTileSize!(self:WBP Discover Tile Base Object Reference=@n2, NewColumnSize:int=@n1, IsSquareTiles:bool=true) -> (NewTileSize:Vector 2D Structure)
     Index="integer > integer"(A=@n1)
    ) -> (ReturnValue:Vector 2D Structure)
   )
  ]}
 SetCompactUIStyle {nodes [
   SetCompactUIStyle() {is=title K2Node_FunctionEntry}
   > Branch(Condition=IsMobileGame())
   > UpdatePrimaryTextScrolling!(execute, self:WBP Discover Tile Base Object Reference=Get WBP_Discover_Tile_Base(), IsMarqueeScrollingEnabled:bool=true) -> (then)
  ]}
 SetImageTexture {nodes [
   n1 = SetImageTexture!() -> (then, LoadingState:EActivityImageLoadingState Enum=Unset) {is=title K2Node_FunctionEntry}
   > n2 = Switch on EActivityImageLoadingState!(execute, Selection:EActivityImageLoadingState Enum=@n1) -> (Unset:Exec, Loading:Exec, Loaded:Exec, Error:Exec) {is=title K2Node_SwitchEnum}
   n3 = Get WBP_Creator_AvatarImage() -> (WBP_Creator_AvatarImage:WBP Creator Avatar Image Object Reference)
   SetCreatorAvatarImageTexture!(
    execute=[@n2.Error, @n2.Loaded, @n2.Unset]
    self:WBP Creator Avatar Image Object Reference=@n3
    CreatorAvatarTexture:Texture Object Reference=GetAvatarImage!(self:Creator View Model Object Reference=Get FortPlayerPageProfileVM()) -> (ReturnValue:Texture Object Reference)
   ) -> (then)
   SetLoadingState(execute=@n2.Loading, self:WBP Creator Avatar Image Object Reference=@n3)
  ]}
 SetIsFollowing {nodes [
   n1 = SetIsFollowing!() -> (then, IsFollowing:bool=~) {is=title K2Node_FunctionEntry}
   > AddFollowUnfollowToContextMenu!(execute, self, IsFollowing:bool=@n1) -> (then) {is=title K2Node_CallFunction}
   > Branch(Condition=GetIsCreatorFollowRebrandingEnabled(self=GetRuntimeOptions()))
   > SetVisibility(
    self=Get WBP_PlusFollowButton()
    InVisibility=Select(Option 0:ESlateVisibility Enum=SelfHitTestInvisible, Option 1:ESlateVisibility Enum=Collapsed, Index=@n1) -> (ReturnValue:ESlateVisibility Enum)
   )
  ]}
 SetIsPlayingNow {nodes [
   n1 = SetIsPlayingNow!() -> (then, IsPlaying:bool=~) {is=title K2Node_FunctionEntry}
   > SetIsPlayingNow!(execute, self:WBP Creator Avatar Image Object Reference=Get WBP_Creator_AvatarImage(), IsPlayingNow:bool=@n1) -> (then) {is=title K2Node_CallFunction}
   > UpdateSecondaryTextVisibility!(execute, self:WBP Discover Tile Base Object Reference=Get WBP_Discover_Tile_Base(), IsVisible:bool=@n1) -> (then)
  ]}
]
variables [
 Action to Remove {type=string, default=ToggleFollowing}
 ContextMenuFollowName {type=string, default=ToggleFollowing}
 Favorited_Sound {type=SoundBase, default=/Game/Sounds/Fort_UI_Sounds/FrontEnd/MultiProduct/Common/Buttons/MSS_UI_MP_Button_Click_Favorite}
 Unfavorited_Sound {type=SoundBase, default=/Game/Sounds/Fort_UI_Sounds/FrontEnd/MultiProduct/Common/Buttons/MSS_UI_MP_Button_Click_Unfavorite}
]
widgets {
 info {
  inheritedWidgetCount=0
  namedSlotCount=2
  parentClass=/DiscoveryBrowser/Tiles/WBP_Discover_Tile_Button.WBP_Discover_Tile_Button_C
  rootWidgetClass=null
  widgetCount=4
 }
 widgets [
  Creator {widgetClassPath=&Overlay, widget=Creator}
  WBP_Creator_AvatarImage {
   widgetClassPath=/DiscoveryBrowser/CreatorProfile/WBP_Creator_AvatarImage
   bIsVariable=true
   parent=Creator
   widget=WBP_Creator_AvatarImage
   properties {creatorAvatarSize=<136 136>}
   slot {class=OverlaySlot, properties {horizontalAlignment=HAlign_Fill, verticalAlignment=VAlign_Fill}}
  }
  Overlay_0 {widgetClassPath=&Overlay, widget=Overlay_0}
  WBP_PlusFollowButton {
   widgetClassPath=&/WBP_Creator_FollowingIndicator
   bIsVariable=true
   parent=Overlay_0
   widget=WBP_PlusFollowButton
   slot {class=OverlaySlot, properties {verticalAlignment=VAlign_Bottom, padding=<ltrb 0 0 0 15>}}
  }
 ]
}
